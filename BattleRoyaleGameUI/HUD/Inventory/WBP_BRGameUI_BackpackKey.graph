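# Battle Royale HUD element widget that shows the backpack icon beside the key bound to toggling the inventory.
# A bottom-right aligned row holds a global-action keybind widget for ToggleInventory and the backpack image.
# An overlay stacks a TutorialHighlight on top, using a custom square outline widget that is scaled and shifted to frame the prompt.
# The event graph is empty, so the Blueprint itself contains only layout.
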WidgetBlueprint /BattleRoyaleGameUI/HUD/Inventory/WBP_BRGameUI_BackpackKey
graphs [EventGraph {links [], nodes []}]
variables []
widgets {
 info {inheritedWidgetCount=1, parentClass=&FortHUDElementWidget, rootWidgetClass=&HorizontalBox, widgetCount=6}
 widgets [
  HorizontalBox_0 {widgetClassPath=&HorizontalBox, widget=HorizontalBox_0}
  Overlay_1 {widgetClassPath=&Overlay, parent=HorizontalBox_0, widget=Overlay_1, slot {class=HorizontalBoxSlot}}
  InventoryRow {
   widgetClassPath=&HorizontalBox
   parent=Overlay_1
   widget=InventoryRow
   slot {class=OverlaySlot, properties {horizontalAlignment=HAlign_Right, verticalAlignment=VAlign_Bottom}}
  }
  KeybindInventory {
   widgetClassPath=/Game/UI/InGame/KeybindWidget/GlobalActionKeybindWidget
   parent=InventoryRow
   widget=KeybindInventory
   properties {globalAction=ToggleInventory}
   slot {class=HorizontalBoxSlot, properties {verticalAlignment=VAlign_Center}}
  }
  Backpack {
   widgetClassPath=&Image
   bIsVariable=true
   parent=InventoryRow
   widget=Backpack
   properties {brush {
     bIsDynamicallyLoaded=false
     drawAs=Image
     imageType=NoImage
     mirroring=NoMirror
     resourceName=None
     resourceObject=/Game/Athena/HUD/Atlas/Shared/T-Icon-Backpack-64-S_Sprite
     tiling=NoTile
     imageSize=<32 32>
     margin=<ltrb 0 0 0 0>
     outlineSettings {
      bUseBrushTransparency=false
      roundingType=HalfHeightRadius
      width=0
      color {colorUseRule=UseColor_Specified, specifiedColor=<rgba 0 0 0 0>}
      cornerRadii=<0 0 0 0>
     }
     tintColor {colorUseRule=UseColor_Specified, specifiedColor=<rgba 1 1 1 1>}
     uVRegion {bIsValid=false, max=<0 0>, min=<0 0>}
    }}
   slot {class=HorizontalBoxSlot, properties {verticalAlignment=VAlign_Center, padding=<ltrb 4 0 0 0>}}
  }
  TutorialHighlight {
   widgetClassPath=&TutorialHighlight
   bIsVariable=true
   parent=Overlay_1
   widget=TutorialHighlight
   properties {
    customHighlightWidget=/Game/Athena/HUD/Tutorial/Tutorial_Outline_Square.Tutorial_Outline_Square_C
    renderTransform {angle=0, scale=<1.3 0.75>, shear=<0 0>, translation=<0 23>}
   }
   slot {class=OverlaySlot, properties {horizontalAlignment=HAlign_Fill, verticalAlignment=VAlign_Fill}}
  }
 ]
}
